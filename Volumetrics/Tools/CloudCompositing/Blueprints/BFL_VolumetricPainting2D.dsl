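; Blueprint function library for the 2D volumetric painting cloud compositing tool.
; Its IsPlayingVolumetricPaintingGameMode function casts the current game mode to VolumetricPaintingGameMode.
; It returns true when the cast succeeds and false otherwise.
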(fn IsPlayingVolumetricPaintingGameMode
  (Utilities|Casting|CastToVolumetricPaintingGameMode (Game|GetGameMode 0)
    (:then
      (return true))
    (:CastFailed
      (return false))))
