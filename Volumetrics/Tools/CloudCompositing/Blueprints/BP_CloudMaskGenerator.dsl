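; BP_CloudMaskGenerator is an actor that composites BP_CloudMask_Object actors into a cloud layout mask for volumetric clouds and an optional sun light function.
; RenderClouds, run on construction and after BeginPlay, draws each object's material by transform, optionally tiled, blending through ping-pong render targets before a final blur.
; Construction feeds the result to cloud and light-function MIDs.
; Helpers save it as a texture or snap owned objects to layer height.
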(event EventTick
  (params (DeltaSeconds :type Float%20%28single-precision%29))
  (if true
    _))

(event EventBeginPlay
  (Utilities|FlowControl|Delay 0.025)
  (CallFunction|RenderClouds))

(event Custom|Allocate%20RTs
  (Class|VolumetricsBPLIbrary|CreateAndAssignRT (Variables|Default|GetScene%20Capture%20RT) (Variables|Default|GetResolution) "RTF_RGBA8")
  (Class|VolumetricsBPLIbrary|CreateAndAssignRT (Variables|Default|GetCloud%20Single%20Render%20RT) (Variables|Default|GetResolution) "RTF_RGBA8")
  (Class|VolumetricsBPLIbrary|CreateAndAssignRT (Variables|Default|GetCloud%20Scratch%20RT%20A) (Variables|Default|GetResolution) "RTF_RGBA8")
  (Class|VolumetricsBPLIbrary|CreateAndAssignRT (Variables|Default|GetFinalized%20Cloud%20RT) (Variables|Default|GetResolution) "RTF_RGBA8")
  (Class|VolumetricsBPLIbrary|CreateAndAssignRT (Variables|Default|GetCloud%20Scratch%20RT%20B) (Variables|Default|GetResolution) "RTF_RGBA8")
  (Rendering|ClearRenderTarget2D (Variables|Default|GetCloud%20Single%20Render%20RT))
  (Rendering|ClearRenderTarget2D (Variables|Default|GetCloud%20Scratch%20RT%20B))
  (Rendering|ClearRenderTarget2D (Variables|Default|GetCloud%20Scratch%20RT%20A))
  (Rendering|ClearRenderTarget2D (Variables|Default|GetFinalized%20Cloud%20RT))
  (Rendering|ClearRenderTarget2D (Variables|Default|GetContent%20Browser%20RT)))

(event Custom|Save%20Texture
  (EditorScripting|Rendering|RenderTarget2DCreateStaticTexture2DEditorOnly (Variables|Default|GetFinalized%20Cloud%20RT) (Utilities|String|Append (Variables|Default|GetTexture%20Path) (Variables|Default|GetTexture%20Name)) "TC_VectorDisplacementmap"))

(fn RenderClouds
  (bind _returnvalue (Utilities|Time|GetRealTimeSeconds))
  (bind _idx (LocalVariables|Getidx))
  (bind _returnvalue_2 (Math|Integer|%25%28Integer%29 _idx 2))
  (bind _scale_km (Variables|Default|GetScale%20Km))
  (bind _returnvalue_3 (* _scale_km 1e+05))
  (bind _resolution (Variables|Default|GetResolution))
  (bind _v2d (Utilities|IntPointToVector2D _resolution))
  (bind _curcloud (Variables|Default|GetCurCloud))
  (bind _returnvalue_4 (== _returnvalue_2 0))
  (bind _cloud_scratch_rt_a (Variables|Default|GetCloud%20Scratch%20RT%20A))
  (bind _cloud_scratch_rt_b (Variables|Default|GetCloud%20Scratch%20RT%20B))
  (bind _returnvalue_5 (select _returnvalue_4 _cloud_scratch_rt_a _cloud_scratch_rt_b))
  (bind _copycloudmid (Variables|Default|GetCopyCloudMID))
  (bind _finalize_blur_cloud_mid (Variables|Default|GetFinalize%20Blur%20Cloud%20MID))
  (bind _finalized_cloud_rt (Variables|Default|GetFinalized%20Cloud%20RT))
  (bind _curcloud_2 (Variables|Default|GetCurCloud))
  (bind _generator (Variables|/Volumetrics/Tools/CloudCompositing/Blueprints/BP_CloudMask_Object.BP_CloudMask_Object_C|GetGenerator _curcloud_2))
  (bind _blendcloudmid (Variables|Default|GetBlendCloudMID))
  (bind _blendcloudmid_2 (Variables|Default|GetBlendCloudMID))
  (bind _cloud_single_render_rt (Variables|Default|GetCloud%20Single%20Render%20RT))
  ;[1 1 1] Avoid multiple renders on Load
  (if (not (== _returnvalue (Variables|Default|GetLastRenderTime)))
    (Variables|Default|SetLastRenderTime _returnvalue)
    (CallFunction|AllocateRTs)
    (Rendering|Material|SetScalarParameterValue "/Volumetrics/Tools/CloudCompositing/MPC/MPC_Clouds.MPC_Clouds" "SkyTextureScaleKm" (Variables|Default|GetScale%20Km) :!class /Script/Engine.KismetMaterialLibrary)
    (Class|VolumetricsBPLIbrary|CreateAndAssignMID (Variables|Default|GetFinalize%20Blur%20Cloud%20MID) "/Volumetrics/Tools/CloudCompositing/Materials/M_Finalize_Cloud_Mask.M_Finalize_Cloud_Mask")
    (Class|VolumetricsBPLIbrary|CreateAndAssignMID (Variables|Default|GetCopyCloudMID) "/Volumetrics/Tools/CloudCompositing/Materials/M_CopyRT.M_CopyRT")
    (Class|VolumetricsBPLIbrary|CreateAndAssignMID _blendcloudmid "/Volumetrics/Tools/CloudCompositing/Materials/M_BlendCloud.M_BlendCloud")
    (Rendering|Material|SetScalarParameterValue _blendcloudmid "AlphaOnlyAffectsRed" (Math|Conversions|ToFloat(Boolean) (Variables|Default|GetAlpha%20Only%20Affects%20Red%20Channel)) :!class /Script/Engine.MaterialInstanceDynamic)
    ;[1 1 1] Loop Over Cloud Objects
    (bind _outactors (Actor|GetAllActorsOfClass "/Volumetrics/Tools/CloudCompositing/Blueprints/BP_CloudMask_Object.BP_CloudMask_Object_C"))
    (for (_array_element _array_index) _outactors
      (bind _output_get (Variables|Default|SetCurCloud _array_element))
      (LocalVariables|Setidx _array_index)
      ; Debug Strng
      (if (Variables|Default|GetDebug%20Print)
        (Development|PrintString (Utilities|GetDisplayName _output_get)))
      (Utilities|IsValid _generator
        (:Is%20Valid
          ;[1 1 1] If CloudObject is attached to a Generator, only use Matching Cloud/Generator
          (if (== self _generator)
            _))
        (:Is%20Not%20Valid
          _))
      ; Setup Render Transform
      (Variables|SetNormalized%20Location (Math|Conversions|ToVector2D(Vector) (+ (/ (Math|Transform|InverseTransformLocation (Math|Transform|MakeTransform "0, 0, 0" (.rotation (Transformation|GetActorTransform))) (Transformation|GetActorLocation _curcloud)) _returnvalue_3) 0.5)))
      (Variables|SetSize (Math|Conversions|ToVector2D(Vector) (* (Transformation|GetActorScale3D _curcloud) 1e+05)))
      (Variables|SetRT%20Screen%20Pos (- (* (Variables|GetNormalized%20Location) (Utilities|IntPointToVector2D (Variables|Default|GetResolution))) (* (* (/ (Variables|GetSize) (Math|Conversions|ToVector2D(Vector) (Math|Conversions|ToVector(Float) _returnvalue_3))) 0.5) _v2d)))
      (Variables|SetRT%20Screen%20Size (* _v2d (/ (Variables|GetSize) (Math|Conversions|ToVector2D(Vector) (* (Math|Conversions|ToVector(Float) (Variables|Default|GetScale%20Km)) 1e+05)))))
      (Variables|SetYaw (.yaw (Transformation|GetActorRotation (Variables|Default|GetCurCloud))))
      (Rendering|ClearRenderTarget2D (Variables|Default|GetCloud%20Single%20Render%20RT))
      (Utilities|IsValid (Variables|Default|GetCloud%20Scratch%20RT%20A)
        (:Is%20Valid
          (Utilities|IsValid (Variables|Default|GetCloud%20Scratch%20RT%20B)
            (:Is%20Valid
              (Utilities|IsValid _cloud_single_render_rt
                (:Is%20Valid
                  (if (Variables|Default|GetMake%20Tiling)
                    (Utilities|FlowControl|Sequence
                      (:then_0
                        ;[1 1 1] Draw Cloud with Tiling Offsets
                        (bind (_canvas _size _context) (Rendering|BeginDrawCanvastoRenderTarget _cloud_single_render_rt))
                        (Utilities|IsValid _canvas
                          (:Is%20Valid
                            (for _array_element_2 (LocalVariables|GetOffsets)
                              (Canvas|DrawMaterial :self _canvas :RenderMaterial (Variables|/Volumetrics/Tools/CloudCompositing/Blueprints/BP_CloudMask_Object.BP_CloudMask_Object_C|GetMID (Variables|Default|GetCurCloud)) :ScreenPosition (+ (LocalVariables|GetRT%20Screen%20Pos) (* _array_element_2 (Utilities|IntPointToVector2D (Variables|Default|GetResolution)))) :ScreenSize (LocalVariables|GetRT%20Screen%20Size) :CoordinateSize "(X=1.000000,Y=1.000000)" :Rotation (LocalVariables|GetYaw) :PivotPoint "(X=0.500000,Y=0.500000)"))
                            (Rendering|EndDrawCanvastoRenderTarget _context))))
                      (:then_1
                        _))
                    (else
                      (Utilities|FlowControl|Sequence
                        (:then_0
                          (bind (_canvas_2 _size_2 _context_2) (Rendering|BeginDrawCanvastoRenderTarget (Variables|Default|GetCloud%20Single%20Render%20RT)))
                          (Utilities|IsValid _canvas_2
                            (:Is%20Valid
                              (Canvas|DrawMaterial :self _canvas_2 :RenderMaterial (Variables|/Volumetrics/Tools/CloudCompositing/Blueprints/BP_CloudMask_Object.BP_CloudMask_Object_C|GetMID (Variables|Default|GetCurCloud)) :ScreenPosition (+ (LocalVariables|GetRT%20Screen%20Pos) (* 0 (Utilities|IntPointToVector2D (Variables|Default|GetResolution)))) :ScreenSize (LocalVariables|GetRT%20Screen%20Size) :CoordinateSize "(X=1.000000,Y=1.000000)" :Rotation (LocalVariables|GetYaw) :PivotPoint "(X=0.500000,Y=0.500000)")
                              (Rendering|EndDrawCanvastoRenderTarget _context_2))))
                        (:then_1
                          _))))
                  ;[1 1 1] Blend Single Cloud to Scratch RT
                  (Rendering|Material|SetTextureParameterValue _blendcloudmid_2 "RT" (Variables|Default|GetCloud%20Single%20Render%20RT))
                  (Rendering|Material|SetTextureParameterValue _blendcloudmid_2 "PreviousRT" (select _returnvalue_4 _cloud_scratch_rt_b _cloud_scratch_rt_a))
                  (Rendering|ClearRenderTarget2D _returnvalue_5)
                  (Rendering|DrawMaterialtoRenderTarget _returnvalue_5 (Variables|Default|GetBlendCloudMID)))))))))
    ;[1 1 1] Finalize Composited Render + Blur
    (Rendering|Material|SetTextureParameterValue _finalize_blur_cloud_mid "Texture" (select (== (Math|Integer|%25%28Integer%29 (LocalVariables|Getidx) 2) 0) (Variables|Default|GetCloud%20Scratch%20RT%20A) (Variables|Default|GetCloud%20Scratch%20RT%20B)))
    (Rendering|Material|SetScalarParameterValue _finalize_blur_cloud_mid "Radius" (Math|Conversions|ToFloat(Integer) (Variables|Default|GetBlur%20Radius)) :!class /Script/Engine.MaterialInstanceDynamic)
    (Utilities|IsValid _finalized_cloud_rt
      (:Is%20Valid
        (Rendering|DrawMaterialtoRenderTarget _finalized_cloud_rt (Variables|Default|GetFinalize%20Blur%20Cloud%20MID))
        (Utilities|FlowControl|Sequence
          (:then_0
            ;[1 1 1] Copy to a Static Render Target
            (if (Variables|Default|GetDraw%20to%20Content%20Browser%20RT)
              (Rendering|Material|SetTextureParameterValue _copycloudmid "RT" (Variables|Default|GetFinalized%20Cloud%20RT))
              (Rendering|DrawMaterialtoRenderTarget (Variables|Default|GetContent%20Browser%20RT) _copycloudmid)))
          (:then_1
            (if (Variables|LightFunction|GetUse%20Lightfunction)
              (Rendering|Material|SetTextureParameterValue (Variables|LightFunction|GetLF%20MID) "Texture" (Variables|Default|GetFinalized%20Cloud%20RT)))))))))

(fn SnapCloudstoZ
  (bind _volumetriccloud (Variables|Default|GetVolumetricCloud))
  (bind _outactors (Actor|GetAllActorsOfClass "/Volumetrics/Tools/CloudCompositing/Blueprints/BP_CloudMask_Object.BP_CloudMask_Object_C"))
  (for _array_element _outactors
    (Utilities|IsValid (Variables|/Volumetrics/Tools/CloudCompositing/Blueprints/BP_CloudMask_Object.BP_CloudMask_Object_C|GetGenerator _array_element)
      (:Is%20Valid
        (if (== (Variables|/Volumetrics/Tools/CloudCompositing/Blueprints/BP_CloudMask_Object.BP_CloudMask_Object_C|GetGenerator _array_element) self)
          (Utilities|IsValid _volumetriccloud
            (:Is%20Valid
              (Transformation|SetActorLocation _array_element (Math|Vector|MakeVector (.x (Transformation|GetActorLocation _array_element)) (.y (Transformation|GetActorLocation _array_element)) (- (* (Variables|/Script/Engine.VolumetricCloudComponent|GetLayerHeight (Variables|/Script/Engine.VolumetricCloud|GetVolumetricCloudComponent _volumetriccloud)) 1e+05) 1e+05)) false false))))))))

(fn ConstructionScript
  (bind _sunlight (Variables|LightFunction|GetSunlight))
  (bind _lightcomponent (Variables|/Script/Engine.Light|GetLightComponent _sunlight))
  (bind _volumetriccloud (Variables|Default|GetVolumetricCloud))
  (bind _cloudmid (Variables|Default|GetCloudMID))
  (bind _lf_mid (Variables|LightFunction|GetLF%20MID))
  (bind _finalized_cloud_rt (Variables|Default|GetFinalized%20Cloud%20RT))
  (Rendering|Material|SetScalarParameterValue "/Volumetrics/Tools/CloudCompositing/MPC/MPC_Clouds.MPC_Clouds" "SkyTextureScaleKm" (Variables|Default|GetScale%20Km) :!class /Script/Engine.KismetMaterialLibrary)
  (Transformation|SetWorldScale3D (Variables|Default|GetStaticMesh) (Math|Conversions|ToVector(Float) (* (Variables|Default|GetScale%20Km) 1e+05)))
  (CallFunction|RenderClouds)
  (Utilities|FlowControl|Sequence
    (:then_0
      (Utilities|IsValid _volumetriccloud
        (:Is%20Valid
          (Landmass|CreateAndAssignMID _cloudmid (Variables|Default|GetCloudMaterial)
            (:Out
              (Rendering|SetMaterial (Variables|/Script/Engine.VolumetricCloud|GetVolumetricCloudComponent _volumetriccloud) _cloudmid)
              (Rendering|Material|SetTextureParameterValue _cloudmid "CloudLayout" (select (Variables|Default|GetApply%20Mask%20to%20Clouds) (Variables|Default|GetFinalized%20Cloud%20RT) (Variables|Default|GetFallback%20Texture)))
              (Rendering|Material|SetScalarParameterValue (Variables|Default|GetCloudMID) "SkyTextureScaleKm" (Variables|Default|GetScale%20Km) :!class /Script/Engine.MaterialInstanceDynamic))))
        (:Is%20Not%20Valid
          _)))
    (:then_1
      _))
  (if (Variables|LightFunction|GetUse%20LightFunction)
    (Utilities|IsValid _sunlight
      (:Is%20Valid
        (Class|VolumetricsBPLIbrary|CreateAndAssignMID _lf_mid (Variables|LightFunction|GetLF%20Material))
        (Rendering|Components|Light|SetLightFunctionMaterial _lightcomponent (Variables|LightFunction|GetLF%20MID))
        (Utilities|IsValid _finalized_cloud_rt
          (:Is%20Valid
            (Rendering|Material|SetTextureParameterValue _lf_mid "Texture" _finalized_cloud_rt))
          (:Is%20Not%20Valid
            _))
        (Rendering|Material|SetVectorParameterValue "/Volumetrics/Tools/CloudCompositing/MPC/MPC_Clouds.MPC_Clouds" "LightDirection" (Math|Conversions|ToLinearColor(Vector) (Math|Vector|Normalize (Transformation|GetForwardVector _lightcomponent))) :!class /Script/Engine.KismetMaterialLibrary)
        (Rendering|Material|SetScalarParameterValue (Variables|LightFunction|GetLF%20MID) "Invert" (Math|Conversions|ToFloat(Boolean) (Variables|LightFunction|GetInvert)) :!class /Script/Engine.MaterialInstanceDynamic)))))
