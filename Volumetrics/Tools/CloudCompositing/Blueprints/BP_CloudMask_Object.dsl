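; BP_CloudMask_Object is a placeable cloud mask element in the volumetric CloudCompositing tool, drawn as either a static mesh or a textured plane.
; Its construction script sets preview visibility and the mesh by draw type; the plane type also gets a dynamic material fed the mask texture, mip, density, alpha, color and noise settings.
; It then triggers RenderClouds on its BP_CloudMaskGenerator, searching the level for one if the reference is invalid, so edits re-render the clouds.
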(fn UpdateMaskGenerator
  (Utilities|IsValid (Variables|Default|GetGenerator)
    (:Is%20Valid
      (Class|BPCloudMaskGenerator|RenderClouds (Variables|Default|GetGenerator)))
    (:Is%20Not%20Valid
      (bind _outactors (Actor|GetAllActorsOfClass "/Volumetrics/Tools/CloudCompositing/Blueprints/BP_CloudMaskGenerator.BP_CloudMaskGenerator_C"))
      (if (> (Utilities|Array|Length _outactors) 0)
        (Utilities|IsValid (Utilities|Array|Get(acopy) _outactors)
          (:Is%20Valid
            (Class|BPCloudMaskGenerator|RenderClouds (Utilities|Array|Get(acopy) _outactors))))))))

(fn ConstructionScript
  (bind _noise_settings (Variables|Default|GetNoise%20Settings))
  (bind (_noisetiling_6_a88c75ce41384502c2167bafb165960a _noiseintensity_7_3c91204e44d26eda1f226691e5dba77c _noiseseed_8_8c33d00144bde4547ad180a7cbd0ec4c) (Utilities|Struct|BreakCloudNoiseSettings _noise_settings))
  (bind _mid (Variables|Debug|GetMID))
  (bind _mid_2 (Variables|Debug|GetMID))
  (Rendering|SetVisibility (Variables|Default|GetPreviewMesh) (Variables|Default|GetShow%20Debug%20Plane) false)
  (Utilities|FlowControl|Sequence
    (:then_0
      (switch Utilities|FlowControl|Switch|SwitchonCloudObjectType (Variables|Default|GetDraw%20Type)
        (:NewEnumerator0
          (Components|StaticMesh|SetStaticMesh (Variables|Default|GetPreviewMesh) (Variables|Default|GetStatic%20Mesh)))
        (:NewEnumerator1
          (Components|StaticMesh|SetStaticMesh (Variables|Default|GetPreviewMesh) "/Engine/ArtTools/RenderToTexture/Meshes/S_1_Unit_Plane.S_1_Unit_Plane")
          (Transformation|SetRelativeScale3D (Variables|Default|GetPreviewMesh) (Math|Conversions|ToVector(Float) 1e+05))
          (Class|VolumetricsBPLIbrary|CreateAndAssignMID _mid (Variables|Default|GetCloud%20Material))
          (Rendering|Material|SetTextureParameterValue _mid "Texture" (Variables|Default|GetCloud%20Mask))
          (Rendering|Material|SetScalarParameterValue _mid "Mip" (Variables|Default|GetMip%20Level) :!class /Script/Engine.MaterialInstanceDynamic)
          (Rendering|Material|SetScalarParameterValue _mid "Opacity" (Variables|Default|GetDensity) :!class /Script/Engine.MaterialInstanceDynamic)
          (Rendering|Material|SetScalarParameterValue _mid "Alpha" (Variables|Default|GetAlpha) :!class /Script/Engine.MaterialInstanceDynamic)
          (Rendering|Material|SetVectorParameterValue _mid "Color" (Math|Conversions|ToLinearColor(Vector) (Variables|Default|GetMask%20Color%20Value)) :!class /Script/Engine.MaterialInstanceDynamic)
          (Rendering|Material|SetScalarParameterValue _mid_2 "Noise Tiling" _noisetiling_6_a88c75ce41384502c2167bafb165960a :!class /Script/Engine.MaterialInstanceDynamic)
          (Rendering|Material|SetScalarParameterValue _mid_2 "Noise Amount" _noiseintensity_7_3c91204e44d26eda1f226691e5dba77c :!class /Script/Engine.MaterialInstanceDynamic)
          (Rendering|Material|SetScalarParameterValue _mid_2 "Seed" _noiseseed_8_8c33d00144bde4547ad180a7cbd0ec4c :!class /Script/Engine.MaterialInstanceDynamic)
          (Rendering|Material|SetMaterial (Variables|Default|GetPreviewMesh) 0 (Variables|Debug|GetMID))
          (Transformation|SetRelativeLocation (Variables|Default|GetSpriteIcon) "0, 0,-2048.000000" false false))))
    (:then_1
      (CallFunction|UpdateMaskGenerator))))
