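; Editor tool Blueprint that bakes a color curve's four channels into an accumulated distance-field render target, previews it on a static mesh, and saves it as a texture.
; Draw runs on construction and after a short delay on BeginPlay, canvas-drawing each channel's samples into a one-pixel-high seed target.
; A material expands seeds into a line-equation target, the 2D jump-flood component processes it, and a composite material writes the result using a channel mask.
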(event Custom|Clear
  (Rendering|ClearRenderTarget2D (Variables|Default|GetAccumulated%20Distance%20Fields))
  (Rendering|ClearRenderTarget2D (Variables|Default|GetCurve%20Line%20Equation%20RT))
  (Rendering|ClearRenderTarget2D (Variables|Default|GetJumpFlood%20Result%20RT))
  (Rendering|ClearRenderTarget2D (Variables|Default|GetSeed%20Curve%20RT)))

(event EventBeginPlay
  (Utilities|FlowControl|Delay 0.035)
  (CallFunction|Draw))

(event Custom|Draw
  (bind _resolution (Variables|Default|GetResolution))
  (bind _x (Utilities|Struct|BreakIntPoint _resolution))
  (bind (_x_2 _y) (Utilities|Struct|BreakIntPoint (Variables|Default|GetResolution)))
  (bind _returnvalue (- _x 1))
  (bind _preview_mid (Variables|Default|GetPreview%20MID))
  (bind _composite_result_mid (Variables|Default|GetComposite%20Result%20MID))
  (Class|VolumetricsBPLIbrary|CreateAndAssignMID (Variables|Default|GetPreview%20MID) "/Volumetrics/Tools/CloudCompositing/Materials/M_Preview_DistanceFieldProfiles.M_Preview_DistanceFieldProfiles")
  (Class|VolumetricsBPLIbrary|CreateAndAssignMID (Variables|Default|GetComposite%20Result%20MID) "/Volumetrics/Tools/CloudCompositing/Materials/M_Combine_DF_Curves_01.M_Combine_DF_Curves_01")
  (Class|VolumetricsBPLIbrary|CreateAndAssignMID (Variables|Default|GetSeed%20Curve%20MID) "/Volumetrics/Tools/CloudCompositing/Materials/M_SeedCurvePositions.M_SeedCurvePositions")
  (Class|VolumetricsBPLIbrary|CreateAndAssignRT (Variables|Default|GetSeed%20Curve%20RT) (Utilities|Struct|MakeIntPoint (Utilities|Struct|BreakIntPoint (Variables|Default|GetResolution)) 1) "RTF_RGBA16f")
  (Class|VolumetricsBPLIbrary|CreateAndAssignRT (Variables|Default|GetCurve%20Line%20Equation%20RT) (Variables|Default|GetResolution) "RTF_RGBA16f")
  (CallFunction|Clear)
  (Utilities|CheckRTsValid
    (:Valid
      (for _index (range 4)
        (Variables|Default|SetChannel _index)
        (Rendering|ClearRenderTarget2D (Variables|Default|GetSeed%20Curve%20RT))
        (Rendering|ClearRenderTarget2D (Variables|Default|GetCurve%20Line%20Equation%20RT))
        (Rendering|ClearRenderTarget2D (Variables|Default|GetJumpFlood%20Result%20RT))
        (bind (_canvas _size _context) (Rendering|BeginDrawCanvastoRenderTarget (Variables|Default|GetSeed%20Curve%20RT)))
        (Utilities|IsValid _canvas
          (:Is%20Valid
            (for _index_2 (range _x)
              (Canvas|DrawBox _canvas (Math|Vector2D|MakeVector2D (+ (Math|Conversions|ToFloat(Integer) _index_2) 0.5) 0.5) (Math|Vector2D|MakeVector2D 1.0 (Math|Conversions|ToFloat(Integer) _y)) 1.0 (Utilities|Struct|MakeLinearColor (/ (+ (Math|Conversions|ToFloat(Integer) _index_2) 0.5) (Math|Conversions|ToFloat(Integer) _x)) (- (Math|Float|Max(Float) (Utilities|Select :Index (Variables|Default|GetChannel) :Option%200 (.A (Utilities|Struct|BreakLinearColor (Math|Curves|GetLinearColorValue (Variables|Default|GetCurve) (/ (Math|Conversions|ToFloat(Integer) _index_2) (Math|Conversions|ToFloat(Integer) (Math|Integer|Max(Integer) _returnvalue 1)))))) :Option%201 (Utilities|Struct|BreakLinearColor (Math|Curves|GetLinearColorValue (Variables|Default|GetCurve) (/ (Math|Conversions|ToFloat(Integer) _index_2) (Math|Conversions|ToFloat(Integer) (Math|Integer|Max(Integer) _returnvalue 1))))) :Option%202 (.G (Utilities|Struct|BreakLinearColor (Math|Curves|GetLinearColorValue (Variables|Default|GetCurve) (/ (Math|Conversions|ToFloat(Integer) _index_2) (Math|Conversions|ToFloat(Integer) (Math|Integer|Max(Integer) _returnvalue 1)))))) :Option%203 (.B (Utilities|Struct|BreakLinearColor (Math|Curves|GetLinearColorValue (Variables|Default|GetCurve) (/ (Math|Conversions|ToFloat(Integer) _index_2) (Math|Conversions|ToFloat(Integer) (Math|Integer|Max(Integer) _returnvalue 1))))))) -1e-05) (Variables|Default|GetEncoding%20Bias)))))
            (Rendering|EndDrawCanvastoRenderTarget _context)
            (Rendering|Material|SetTextureParameterValue (Variables|Default|GetSeed%20Curve%20MID) "RT" (Variables|Default|GetSeed%20Curve%20RT))
            (Rendering|DrawMaterialtoRenderTarget (Variables|Default|GetCurve%20Line%20Equation%20RT) (Variables|Default|GetSeed%20Curve%20MID))
            (bind _result_rt (Class|JumpFloodComponent2D|JumpFlood (Variables|Default|GetJumpFloodComponent_2D) "NewEnumerator1" (Variables|Default|GetCurve%20Line%20Equation%20RT) 0 :!class /Volumetrics/Tools/JumpFlooding/Blueprints/JumpFloodComponent_2D.JumpFloodComponent_2D_C))
            (bind _astexture_render_target_2d (Utilities|Casting|CastToTextureRenderTarget2D _result_rt))
            (Variables|Default|SetJumpFlood%20Result%20RT _astexture_render_target_2d)
            (Rendering|Material|SetVectorParameterValue _preview_mid "TextureSize" (Utilities|IntPointToLinearColor (Variables|Default|GetResolution)) :!class /Script/Engine.MaterialInstanceDynamic)
            (Rendering|Material|SetTextureParameterValue _preview_mid "RT" (Variables|Default|GetAccumulated%20Distance%20Fields))
            (Rendering|Material|SetMaterial (Variables|Default|GetStaticMesh) 0 (Variables|Default|GetPreview%20MID))
            (Rendering|Material|SetVectorParameterValue _composite_result_mid "Channel" (Utilities|Select :Index (Variables|Default|GetChannel) :Option%200 "(R=0.000000,G=0.000000,B=0.000000,A=1.000000)" :Option%201 "(R=1.000000,G=0.000000,B=0.000000,A=0.000000)" :Option%202 "(R=0.000000,G=1.000000,B=0.000000,A=0.000000)" :Option%203 "(R=0.000000,G=0.000000,B=1.000000,A=0.000000)") :!class /Script/Engine.MaterialInstanceDynamic)
            (Rendering|Material|SetTextureParameterValue _composite_result_mid "RT" (Variables|Default|GetJumpFlood%20Result%20RT))
            (Rendering|DrawMaterialtoRenderTarget (Variables|Default|GetAccumulated%20Distance%20Fields) (Variables|Default|GetComposite%20Result%20MID))))))))

(event Custom|Save
  (EditorScripting|Rendering|RenderTarget2DCreateStaticTexture2DEditorOnly (Variables|Default|GetAccumulated%20Distance%20Fields) (Variables|Default|GetTexture%20Name) "TC_HDR"))

(fn ConstructionScript
  (CallFunction|Draw))
