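; BP_PaintClouds is the actor for painting volumetric cloud density and flowmap velocity into render targets over a world-space square.
; Draw maps the brush's world location to UV and renders a brush material, reading previous-frame RT copies to accumulate strokes.
; Paint_Pawn receives this actor in the painting game mode; editor-only Save bakes RTs to textures, and Load restores them.
; It also drives cloud MPC values, an optional sun light function and interface hand-off.
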(event Custom|OnPossessDelegate_Event_0
  (params (InPawn :type Pawn%20Object%20Reference))
  ;[0 0 0] #Todo only works *sometimes* why?
  (bind _aspaint_pawn (Utilities|Casting|CastToPaint_Pawn InPawn))
  (Class|PaintPawn|InitPaintCloudsBP _aspaint_pawn self))

(event EventBeginPlay
  (Utilities|FlowControl|Sequence
    (:then_0
      (if (Class|BFLVolumetricPainting2D|IsPlayingVolumetricPaintingGameMode)
        (Utilities|FlowControl|Sequence
          (:then_0
            ;[0 0 0] Init game mode #Todo use delegate
            (bind _aspaint_pawn (Utilities|Casting|CastToPaint_Pawn (Pawn|GetControlledPawn (Game|GetPlayerController 0)))
              (:then
                _)
              (:CastFailed
                (Utilities|FlowControl|Delay 1.0)))
            (Class|PaintPawn|InitPaintCloudsBP _aspaint_pawn self))
          (:then_1
            ;[0 0 0] Init RT prev
            (bind _returnvalue (Rendering|CreateRenderTarget2D (Variables|/Script/Engine.TextureRenderTarget2D|GetSizeX (Variables|Default|RenderTargets|GetCloud%20Mask%20RT)) (Variables|/Script/Engine.TextureRenderTarget2D|GetSizeY (Variables|Default|RenderTargets|GetCloud%20Mask%20RT)) (Variables|/Script/Engine.TextureRenderTarget2D|GetRenderTargetFormat (Variables|Default|RenderTargets|GetCloud%20Mask%20RT)) "(R=0.000000,G=0.000000,B=0.000000,A=1.000000)" true))
            (Variables|Default|RenderTargets|SetCloudMaskRTPrev _returnvalue)
            (bind _returnvalue_2 (Rendering|CreateRenderTarget2D (Variables|/Script/Engine.TextureRenderTarget2D|GetSizeX (Variables|Default|RenderTargets|GetFlowmap%20RT)) (Variables|/Script/Engine.TextureRenderTarget2D|GetSizeY (Variables|Default|RenderTargets|GetFlowmap%20RT)) (Variables|/Script/Engine.TextureRenderTarget2D|GetRenderTargetFormat (Variables|Default|RenderTargets|GetFlowmap%20RT)) "(R=0.000000,G=0.000000,B=0.000000,A=1.000000)" true))
            (Variables|Default|RenderTargets|SetFlowmapRTPrev _returnvalue_2)
            (bind _returnvalue_3 (Rendering|Material|CreateDynamicMaterialInstance "/Volumetrics/Tools/CloudCompositing/Blueprints/GameSetup/M_CopyRT.M_CopyRT" "None" "None" :!class /Script/Engine.KismetMaterialLibrary))
            (Variables|Default|SetMID_CopyRT _returnvalue_3)
            (CallFunction|CopytoRTPreviousBuffer)))))
    (:then_2
      (Rendering|Material|SetScalarParameterValue "/Volumetrics/Tools/CloudCompositing/MPC/MPC_Clouds.MPC_Clouds" "SkyTextureScaleKm" (Variables|Default|GetCloud%20Scale%20Km) :!class /Script/Engine.KismetMaterialLibrary)
      (CallFunction|SetLightFunction))))

(event Custom|Draw
  (params (BlendMode :type CloudPaintBlendMode%20Enum :default "NewEnumerator0"))
  (bind (_uv_x _uv_y) (Math|Vector2D|BreakVector2D (CallFunction|WorldLocationtoUV (Variables|Default|GetPaint%20World%20Location))))
  (bind _paint_mode (Variables|Default|GetPaint%20Mode))
  (bind _flowmap_rt (Variables|Default|RenderTargets|GetFlowmap%20RT))
  (bind _cloud_mask_rt (Variables|Default|RenderTargets|GetCloud%20Mask%20RT))
  (bind _returnvalue (Utilities|Select :Index _paint_mode :NewEnumerator0 _cloud_mask_rt :NewEnumerator1 _flowmap_rt))
  (bind _mid (Variables|Default|GetMID))
  (bind _mid_2 (Variables|Default|GetMID))
  (Rendering|Material|SetScalarParameterValue _mid_2 "IsAddMode" (Utilities|Select :Index BlendMode :NewEnumerator0 1.0 :NewEnumerator1 0.0) :!class /Script/Engine.MaterialInstanceDynamic)
  (Rendering|Material|SetScalarParameterValue _mid_2 "AllowNegative" (Utilities|Select :Index (Variables|Default|GetPaint%20Mode) :NewEnumerator0 0.0 :NewEnumerator1 1.0) :!class /Script/Engine.MaterialInstanceDynamic)
  (Rendering|Material|SetVectorParameterValue _mid_2 "ColorChannel" (Utilities|Select :Index (Variables|Default|GetColor%20Channel) :NewEnumerator0 "(R=1.000000,G=0.000000,B=0.000000,A=0.000000)" :NewEnumerator1 "(R=0.000000,G=1.000000,B=0.000000,A=0.000000)" :NewEnumerator2 "(R=0.000000,G=0.000000,B=1.000000,A=0.000000)" :NewEnumerator3 "(R=0.000000,G=0.000000,B=0.000000,A=1.000000)") :!class /Script/Engine.MaterialInstanceDynamic)
  (Rendering|Material|SetTextureParameterValue _mid "PreviousBuffer" (Utilities|Select :Index (Variables|Default|GetPaint%20Mode) :NewEnumerator0 (Variables|Default|RenderTargets|GetCloudMaskRTPrev) :NewEnumerator1 (Variables|Default|RenderTargets|GetFlowmapRTPrev)))
  (Rendering|Material|SetVectorParameterValue _mid "P" (Math|Color|MakeColor _uv_x _uv_y) :!class /Script/Engine.MaterialInstanceDynamic)
  (Rendering|Material|SetScalarParameterValue _mid "R" (Variables|Default|GetBrush%20Size) :!class /Script/Engine.MaterialInstanceDynamic)
  (Rendering|Material|SetScalarParameterValue _mid "Strength" (Variables|Default|GetBrush%20Strength) :!class /Script/Engine.MaterialInstanceDynamic)
  (Rendering|Material|SetScalarParameterValue _mid "PaintVelocity" (Utilities|Select :Index (Variables|Default|GetPaint%20Mode) :NewEnumerator0 0.0 :NewEnumerator1 1.0) :!class /Script/Engine.MaterialInstanceDynamic)
  (Rendering|Material|SetVectorParameterValue _mid "Velocity" (Math|Conversions|ToLinearColor(Vector) (* (Math|Vector|SelectVector (Variables|Default|GetMeasuredVelocity) (/ (* (- (Variables|Default|GetPaint%20World%20Location) (Variables|Default|GetPrevious%20World%20Paint%20Location)) "1.000000,1.000000,0") (* (Variables|Default|GetCloud%20Scale%20Km) 1e+05)) (> (Math|Vector|DotProduct (Variables|Default|GetMeasuredVelocity) (Variables|Default|GetMeasuredVelocity)) 0)) (Variables|Default|GetBrush%20Strength))) :!class /Script/Engine.MaterialInstanceDynamic)
  (Rendering|ClearRenderTarget2D _returnvalue)
  (Rendering|DrawMaterialtoRenderTarget _returnvalue (Variables|Default|GetMID))
  (CallFunction|CopytoRTPreviousBuffer))

(event Custom|Clear%20Density
  (bind _cloud_mask_rt (Variables|Default|RenderTargets|GetCloud%20Mask%20RT))
  (Rendering|ClearRenderTarget2D _cloud_mask_rt)
  (Rendering|Material|SetTextureParameterValue (Variables|Default|GetSeed%20MID) "RT" (Variables|Default|GetSeed%20Density%20Texture))
  (Rendering|DrawMaterialtoRenderTarget _cloud_mask_rt (Variables|Default|GetSeed%20MID))
  (CallFunction|CopytoRTPreviousBuffer))

(event Custom|Clear%20Velocity
  (bind _flowmap_rt (Variables|Default|RenderTargets|GetFlowmap%20RT))
  (Rendering|ClearRenderTarget2D _flowmap_rt)
  (Rendering|Material|SetTextureParameterValue (Variables|Default|GetSeed%20MID) "RT" (Variables|Default|GetSeed%20Velocity%20Texture))
  (Rendering|DrawMaterialtoRenderTarget _flowmap_rt (Variables|Default|GetSeed%20MID))
  (CallFunction|CopytoRTPreviousBuffer))

(event EventTick
  (params (DeltaSeconds :type Float%20%28single-precision%29))
  (bind _curvesbp (Variables|Default|GetCurvesBP))
  (Utilities|FlowControl|Sequence
    (:then_0
      (CallFunction|UpdatePreviewValues)
      (if (Variables|Default|GetUpdate%20Curves%20Dynamically)
        (Utilities|IsValid _curvesbp
          (:Is%20Valid
            (Class|BPCurveProfiletoRT|Draw _curvesbp)))))
    (:then_1
      (if (Class|BFLVolumetricPainting2D|IsPlayingVolumetricPaintingGameMode)
        (CallFunction|UpdateBrushNoise)))))

(event Custom|Update%20Curves
  (bind _curvesbp (Variables|Default|GetCurvesBP))
  (Utilities|IsValid _curvesbp
    (:Is%20Valid
      (Class|BPCurveProfiletoRT|Draw _curvesbp))))

(fn CopyToRTPreviousBuffer
  (bind _mid_copyrt (Variables|Default|GetMID_CopyRT))
  (Rendering|ClearRenderTarget2D (Variables|Default|RenderTargets|GetCloudMaskRTPrev))
  (Rendering|ClearRenderTarget2D (Variables|Default|RenderTargets|GetFlowmapRTPrev))
  (Rendering|Material|SetTextureParameterValue _mid_copyrt "RT" (Variables|Default|RenderTargets|GetCloud%20Mask%20RT))
  (Rendering|DrawMaterialtoRenderTarget (Variables|Default|RenderTargets|GetCloudMaskRTPrev) _mid_copyrt)
  (Rendering|Material|SetTextureParameterValue _mid_copyrt "RT" (Variables|Default|RenderTargets|GetFlowmap%20RT))
  (Rendering|DrawMaterialtoRenderTarget (Variables|Default|RenderTargets|GetFlowmapRTPrev) _mid_copyrt))

(fn GetVolumetricPainting2D
  (return (Utilities|Struct|MakeVolumetricPainting2D (Variables|Default|RenderTargets|GetCloud%20Mask%20RT) (Variables|Default|RenderTargets|GetFlowmap%20RT) (Math|Box2D|MakeBox2D (Math|Conversions|ToVector2D(Vector) (CallFunction|UVToWorldLocation)) (Math|Conversions|ToVector2D(Vector) (CallFunction|UVToWorldLocation "(X=1.000000,Y=1.000000)"))))))

(fn GetVolumetricPainting2DStaticTexture
  (return (Utilities|Struct|MakeVolumetricPainting2D (Variables|Save/Load|GetDensityStaticTexture) (Variables|Save/Load|GetFlowmapStaticTexture) (Math|Box2D|MakeBox2D (Math|Conversions|ToVector2D(Vector) (CallFunction|UVToWorldLocation)) (Math|Conversions|ToVector2D(Vector) (CallFunction|UVToWorldLocation "(X=1.000000,Y=1.000000)"))))))

(fn Load
  (bind _cloud_mask_rt (Variables|Default|RenderTargets|GetCloud%20Mask%20RT))
  (bind _flowmap_rt (Variables|Default|RenderTargets|GetFlowmap%20RT))
  (bind _returnvalue (Rendering|Material|CreateDynamicMaterialInstance "/Volumetrics/Tools/CloudCompositing/Materials/M_CopyRT.M_CopyRT" "None" "None" :!class /Script/Engine.KismetMaterialLibrary))
  (Rendering|Material|SetTextureParameterValue _returnvalue "RT" (Variables|Save/Load|GetDensityStaticTexture))
  (Rendering|ClearRenderTarget2D _cloud_mask_rt)
  (Rendering|DrawMaterialtoRenderTarget _cloud_mask_rt _returnvalue)
  (Rendering|Material|SetTextureParameterValue _returnvalue "RT" (Variables|Save/Load|GetFlowmapStaticTexture))
  (Rendering|ClearRenderTarget2D _flowmap_rt)
  (Rendering|DrawMaterialtoRenderTarget _flowmap_rt _returnvalue)
  (CallFunction|CopytoRTPreviousBuffer))

(fn ReadPaintingRTValueAtWorldPosition
  (params (WorldPosition :type Vector))
  (bind (_uv_x _uv_y) (Math|Vector2D|BreakVector2D (CallFunction|WorldLocationtoUV WorldPosition)))
  (bind (_volumetricpainting2d_density_17_b70cf61d4bfbfef73ada7ea102be0447 _volumetricpainting2d_flowmap_18_f50e0c55461b080fc0d323bbc69297d2) (Utilities|Struct|BreakVolumetricPainting2D (CallFunction|GetVolumetricPainting2D)))
  (Utilities|FlowControl|Sequence
    (:then_0
      (bind _astexture_render_target_2d (Utilities|Casting|CastToTextureRenderTarget2D _volumetricpainting2d_density_17_b70cf61d4bfbfef73ada7ea102be0447))
      (bind _returnvalue (Rendering|ReadRenderTargetRawUV _astexture_render_target_2d _uv_x _uv_y false))
      (LocalVariables|SetDensityValue _returnvalue))
    (:then_1
      (bind _astexture_render_target_2d_2 (Utilities|Casting|CastToTextureRenderTarget2D _volumetricpainting2d_flowmap_18_f50e0c55461b080fc0d323bbc69297d2))
      (bind _returnvalue_2 (Rendering|ReadRenderTargetRawUV _astexture_render_target_2d_2 _uv_x _uv_y false))
      (LocalVariables|SetFlowmapValue _returnvalue_2))
    (:then_2
      (return (LocalVariables|GetDensityValue) (LocalVariables|GetFlowmapValue)))))

(fn Save
  (EditorScripting|Rendering|ConvertRenderTargettoTexture2DEditorOnly (Variables|Default|RenderTargets|GetCloud%20Mask%20RT) (Variables|Save/Load|GetDensityStaticTexture))
  (EditorScripting|Rendering|ConvertRenderTargettoTexture2DEditorOnly (Variables|Default|RenderTargets|GetFlowmap%20RT) (Variables|Save/Load|GetFlowmapStaticTexture)))

(fn SetLightFunction
  (bind _sunlight (Variables|LightFunction|GetSunlight))
  (bind _lightcomponent (Variables|/Script/Engine.Light|GetLightComponent _sunlight))
  (bind _cloud_mask_rt (Variables|Default|RenderTargets|GetCloud%20Mask%20RT))
  (bind _lf_mid (Variables|LightFunction|GetLF%20MID))
  (Variables|LightFunction|GetSunlight
    (:then
      (if (Variables|LightFunction|GetUseLightFunction)
        (Class|VolumetricsBPLIbrary|CreateAndAssignMID _lf_mid (Variables|LightFunction|GetLF%20Material))
        (Rendering|Components|Light|SetLightFunctionMaterial _lightcomponent (Variables|LightFunction|GetLF%20MID))
        (Utilities|IsValid _cloud_mask_rt
          (:Is%20Valid
            (Rendering|Material|SetTextureParameterValue _lf_mid "Texture" _cloud_mask_rt))
          (:Is%20Not%20Valid
            _))
        (Rendering|Material|SetVectorParameterValue "/Volumetrics/Tools/CloudCompositing/MPC/MPC_Clouds.MPC_Clouds" "LightDirection" (Math|Conversions|ToLinearColor(Vector) (Math|Vector|Normalize (Transformation|GetForwardVector _lightcomponent))) :!class /Script/Engine.KismetMaterialLibrary)
        (Rendering|Material|SetScalarParameterValue (Variables|LightFunction|GetLF%20MID) "Invert" (Math|Conversions|ToFloat(Boolean) (Variables|LightFunction|GetInvert)) :!class /Script/Engine.MaterialInstanceDynamic)
        (else
          (Rendering|Components|Light|SetLightFunctionMaterial _lightcomponent 0))))
    (:else)))

(fn SetMeasuredVelocity
  (params (MeasuredVelocity :type Vector))
  (Variables|Default|SetMeasuredVelocity MeasuredVelocity))

(fn SetPaintingToInterface
  (bind _actortotransferto (Variables|VolumetricPainting2DInterface|GetActorToTransferTo))
  (Utilities|IsValid _actortotransferto
    (:Is%20Valid
      (if (Utilities|DoesObjectImplementInterface _actortotransferto "/Volumetrics/Tools/CloudCompositing/Blueprints/VolumetricPainting2DInterface.VolumetricPainting2DInterface_C")
        (Class|VolumetricPainting2DInterface|SetVolumetricPainting2D(Message) _actortotransferto (CallFunction|GetVolumetricPainting2D))
        (else
          (Development|PrintString "Target actor doesn't implement VolumetricPainting2DInterface" true true "(R=1.000000,G=0.000000,B=0.000000,A=1.000000)" 1e+01))))))

(fn ShouldProjectBrushOnProxyMesh
  (return (Variables|VolumetricPainting2DInterface|GetbShouldProjectBrushOnProxyMesh)))

(fn UpdateBrushNoise
  (bind _brush_noise_settings (Variables|LightFunction|GetBrush%20Noise%20Settings))
  (bind (_noisetiling_9_34683dd54a9bc18d624e82934b0734d9 _noiseamount_10_29ddc73f433104cd28a5f7a595930c24 _randomizenoise_8_52f1be0f4268a855ff5ed297033dadec) (Utilities|Struct|BreakCloudBrushNoiseSettings _brush_noise_settings))
  (bind _mid (Variables|Default|GetMID))
  (Rendering|Material|SetScalarParameterValue _mid "NoiseTiling" _noisetiling_9_34683dd54a9bc18d624e82934b0734d9 :!class /Script/Engine.MaterialInstanceDynamic)
  (Rendering|Material|SetScalarParameterValue _mid "NoiseAmount" _noiseamount_10_29ddc73f433104cd28a5f7a595930c24 :!class /Script/Engine.MaterialInstanceDynamic)
  (Rendering|Material|SetScalarParameterValue _mid "Randomize" (Math|Conversions|ToFloat(Boolean) _randomizenoise_8_52f1be0f4268a855ff5ed297033dadec) :!class /Script/Engine.MaterialInstanceDynamic))

(fn UpdatePaintLocation
  (params (New%20Location :type Vector) (Old%20Location :type Vector))
  (Variables|Default|SetPaint%20World%20Location New%20Location)
  (Variables|Default|SetPrevious%20World%20Paint%20Location Old%20Location))

(fn UpdatePreviewValues
  (Rendering|Material|SetScalarParameterValue "/Volumetrics/Tools/CloudCompositing/MPC/MPC_Clouds.MPC_Clouds" "LightningMultiplier" (* (Variables|Default|GetLightning%20Multiplier) (Math|Conversions|ToFloat(Boolean) (Variables|Default|GetShow%20Lightning))) :!class /Script/Engine.KismetMaterialLibrary)
  (Rendering|Material|SetScalarParameterValue "/Volumetrics/Tools/CloudCompositing/MPC/MPC_Clouds.MPC_Clouds" "NoiseMultiplier" (Variables|Default|GetNoise%20Multiplier) :!class /Script/Engine.KismetMaterialLibrary))

(fn ConstructionScript
  (bind _brush_noise_settings (Variables|LightFunction|GetBrush%20Noise%20Settings))
  (bind (_noisetiling_9_34683dd54a9bc18d624e82934b0734d9 _noiseamount_10_29ddc73f433104cd28a5f7a595930c24 _randomizenoise_8_52f1be0f4268a855ff5ed297033dadec) (Utilities|Struct|BreakCloudBrushNoiseSettings _brush_noise_settings))
  (bind _cloud_scale_km (Variables|Default|GetCloud%20Scale%20Km))
  (bind _returnvalue (* _cloud_scale_km 1e+05))
  (bind _returnvalue_2 (Math|Conversions|ToVector(Float) _returnvalue))
  (bind (_returnvalue_x _returnvalue_y) (Math|Vector|BreakVector (Transformation|GetActorLocation)))
  (bind _mid (Variables|Default|GetMID))
  (bind _returnvalue_3 (Rendering|Material|CreateDynamicMaterialInstance "/Volumetrics/Tools/CloudCompositing/Blueprints/GameSetup/M_AddForce.M_AddForce" "None" "None" :!class /Script/Engine.KismetMaterialLibrary))
  (Variables|Default|SetMID _returnvalue_3)
  (bind _returnvalue_4 (Rendering|Material|CreateDynamicMaterialInstance "/Volumetrics/Tools/CloudCompositing/Materials/M_CopyRT.M_CopyRT" "None" "None" :!class /Script/Engine.KismetMaterialLibrary))
  (Variables|Default|SetSeed%20MID _returnvalue_4)
  (Rendering|SetVisibility (Variables|Default|GetPreview) (Variables|Default|GetShow%20Grid%20Bounds) false)
  (Transformation|SetWorldScale3D (Variables|Default|GetPreview) _returnvalue_2)
  (Transformation|SetWorldScale3D (Variables|Default|GetPreview1) (* _returnvalue_2 "1.000000,1.000000,-1.000000"))
  (Transformation|SetWorldLocation (Variables|Default|GetPreview) (Math|Vector|MakeVector _returnvalue_x _returnvalue_y (* (Variables|Default|GetElevation%20to%20Paint%20at%20Km) 1e+05)) false false)
  (Rendering|Material|SetScalarParameterValue "/Volumetrics/Tools/CloudCompositing/MPC/MPC_Clouds.MPC_Clouds" "SkyTextureScaleKm" (Variables|Default|GetCloud%20Scale%20Km) :!class /Script/Engine.KismetMaterialLibrary)
  (CallFunction|UpdatePreviewValues)
  (CallFunction|SetLightFunction)
  (Rendering|Material|SetScalarParameterValue _mid "NoiseTiling" _noisetiling_9_34683dd54a9bc18d624e82934b0734d9 :!class /Script/Engine.MaterialInstanceDynamic)
  (Rendering|Material|SetScalarParameterValue _mid "NoiseAmount" _noiseamount_10_29ddc73f433104cd28a5f7a595930c24 :!class /Script/Engine.MaterialInstanceDynamic)
  (Rendering|Material|SetScalarParameterValue _mid "Randomize" (Math|Conversions|ToFloat(Boolean) _randomizenoise_8_52f1be0f4268a855ff5ed297033dadec) :!class /Script/Engine.MaterialInstanceDynamic))

(fn UVToWorldLocation
  (params (UV :type Vector%202D%20Structure))
  (bind (_returnvalue_x _returnvalue_y) (Math|Vector2D|BreakVector2D (Utilities|Operators|Multiply (- UV "(X=0.500000,Y=0.500000)") (Math|Vector2D|MakeVector2D (* (Variables|Default|GetCloud%20Scale%20Km) 1e+05) (* (Variables|Default|GetCloud%20Scale%20Km) 1e+05)))))
  (return (Utilities|Operators|Add (Math|Vector|MakeVector _returnvalue_x _returnvalue_y) (Transformation|GetActorLocation))))

(fn WorldLocationToUV
  (params (WorldLocation :type Vector))
  (return (Math|Conversions|ToVector2D(Vector) (+ (/ (- WorldLocation (Transformation|GetActorLocation)) (* (Variables|Default|GetCloud%20Scale%20Km) 1e+05)) 0.5))))
