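; BP_PaintFlowMap is a cloud-compositing tool Blueprint that feeds flow-map brush parameters to a dynamic material.
; Draw converts Force Location into actor-local space, remapping a 500-unit span centered on the actor to 0–1 as "P".
; It also passes Force Velocity as "V" and Brush Size as "R".
; No render-target draw appears here, and Clear and the construction script are empty.
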(event Custom|Draw
  (bind _mid (Variables|Default|GetMID))
  (Rendering|Material|SetVectorParameterValue _mid "P" (Math|Conversions|ToLinearColor(Vector) (+ (/ (Math|Transform|InverseTransformLocation (Transformation|GetActorTransform) (Variables|Default|GetForce%20Location)) 500) 0.5)) :!class /Script/Engine.MaterialInstanceDynamic)
  (Rendering|Material|SetVectorParameterValue _mid "V" (Math|Conversions|ToLinearColor(Vector) (Variables|Default|GetForce%20Velocity)) :!class /Script/Engine.MaterialInstanceDynamic)
  (Rendering|Material|SetScalarParameterValue _mid "R" (Variables|Default|GetBrush%20Size) :!class /Script/Engine.MaterialInstanceDynamic))

(event Custom|Clear)

(fn ConstructionScript)
