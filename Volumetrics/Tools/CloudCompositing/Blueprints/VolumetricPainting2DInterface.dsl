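; Blueprint interface for 2D volumetric painting in the Volumetrics cloud compositing tools.
; It declares a setter that takes a Volumetric Painting 2D structure, plus three Boolean visibility setters for the brush collision mesh, the flowmap indicator and the volumetric effect.
; The file holds only signatures with no outputs or logic, so behavior is left to implementers.
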(fn SetVolumetricPainting2D
  (params (InVolumetricPainting2D :type Volumetric%20Painting%202D%20Structure)))

(fn SetBrushCollisionMeshVisibility
  (params (NewVisibility :type Boolean)))

(fn SetFlowmapIndicatorVisibility
  (params (NewVisibility :type Boolean)))

(fn SetVolumetricEffectVisibility
  (params (NewVisibility :type Boolean)))
