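# Lightning_Spline is a material function that computes a camera-facing width offset for ribbon geometry, named for spline lightning in the Volumetrics cloud compositing effects.
# It crosses the camera-to-position vector with an Axis input and normalizes the result to get a sideways direction.
# That direction is scaled by centered UV U and the Width input, billboarding the strip around its axis.
# A second normalized view vector is left unconnected.
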MaterialFunction /Volumetrics/Tools/CloudCompositing/Effects/Lightning_Spline
graph {nodes [
  n1 = Subtract(A=CameraPositionWS(), B=FunctionInput(Preview=WorldPosition().XYZ) {bUsePreviewValueAsDefault=true, inputName=WorldPosition})
  n2 = Normalize(VectorInput=CrossProduct(A=@n1, B=FunctionInput() {inputName=Axis, sortPriority=1, previewValue=<0 0 1 1>}))
  n3 = Subtract(A=ComponentMask(None=TextureCoordinate()) {r=true}) {constB=0.5}
  FunctionOutput(None=Multiply(A=Multiply(A=@n2, B=@n3), B=FunctionInput() {inputName=Width, inputType=FunctionInput_Scalar, sortPriority=2, previewValue=<0 0 1 1>}))
  Normalize(VectorInput=@n1)
 ]}
