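# M_4ChannelsMax is an unlit material in the volumetrics cloud compositing tools that merges two textures by per-channel maximum.
# It samples parameters RT1 and RT2, by name likely render targets, taking the maximum of their RGB and, separately, of their alpha.
# The alpha maximum is inverted with OneMinus, and the material uses Alpha Composite (premultiplied alpha) blending.
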Material /Volumetrics/Tools/CloudCompositing/Materials/M_4ChannelsMax
graph {
 properties {blendMode=BLEND_AlphaComposite, pixelDepthOffsetMode=PDOM_Legacy, shadingModel=MSM_Unlit, usedShadingModels=Unlit}
 nodes [
  n1 = TextureSampleParameter2D() {parameterName=RT1}
  n2 = TextureSampleParameter2D() {parameterName=RT2}
  Max(A=@n1.RGB, B=@n2.RGB)
  OneMinus(None=Max(A=@n1.A, B=@n2.A))
 ]
}
