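# Unlit alpha-composite material for the volumetric cloud compositing tool that blends the RT texture with PreviousRT.
# It takes the per-channel maximum of both RGB values, alongside an inverted alpha picked from one of the two textures.
# A saturated ceiling of their difference makes a hard mask that selects RT's alpha where it is greater.
# AlphaOnlyAffectsRed shifts that comparison from RGB dot products to red alone.
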Material /Volumetrics/Tools/CloudCompositing/Materials/M_BlendCloud
graph {
 properties {blendMode=BLEND_AlphaComposite, pixelDepthOffsetMode=PDOM_Legacy, shadingModel=MSM_Unlit, usedShadingModels=Unlit}
 nodes [
  n1 = TextureSampleParameter2D() {
   parameterName=PreviousRT
   samplerType=SAMPLERTYPE_LinearColor
   texture=/Engine/Functions/Engine_MaterialFunctions02/PivotPainter2/Black_1x1_EXR_Texture
  }
  n2 = TextureSampleParameter2D() {
   parameterName=RT
   samplerType=SAMPLERTYPE_LinearColor
   texture=&Black_1x1_EXR_Texture
  }
  n3 = Constant() {r=1}
  n4 = Constant() {r=1}
  n5 = LinearInterpolate(
   A=Subtract(A=DotProduct(A=@n3, B=@n2.RGB), B=DotProduct(A=@n4, B=@n1.RGB))
   B=Subtract(A=@n2.R, B=@n1.R)
   Alpha=ScalarParameter() {parameterName=AlphaOnlyAffectsRed}
  )
  LinearInterpolate(A=OneMinus(None=@n1.A), B=OneMinus(None=@n2.A), Alpha=Saturate(None=Ceil(None=@n5)))
  Max(A=@n2.RGB, B=@n1.RGB)
 ]
}
