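# Masked, unlit sprite material in the cloud compositing tools, built around the engine editor's volumetric cloud icon texture.
# A Sprite material function takes its XY scale from a Size parameter multiplied by the object's X scale, so sizing follows that one axis.
# A Brightness parameter multiplies the icon's RGB.
# The graph shows no links from either branch to material outputs.
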Material /Volumetrics/Tools/CloudCompositing/Materials/M_CloudSprite_01
graph {
 properties {blendMode=BLEND_Masked, pixelDepthOffsetMode=PDOM_Legacy, shadingModel=MSM_Unlit, usedShadingModels=Unlit}
 nodes [
  MaterialFunctionCall!(
   Center Location
   0-1 UVs
   XY scale=Multiply(
    A=ScalarParameter() {defaultValue=512, parameterName=Size}
    B=MaterialFunctionCall() {materialFunction=/Engine/Functions/Engine_MaterialFunctions02/WorldPositionOffset/ObjectScale}.Scale X
   )
   OverridePivot
   "Normals (Optional)"
  ) {materialFunction=&Sprite}
  Multiply(A=ScalarParameter() {defaultValue=0.5, parameterName=Brightness}, B=TextureSample() {texture=/Engine/EditorResources/S_VolumetricCloud}.RGB)
 ]
}
