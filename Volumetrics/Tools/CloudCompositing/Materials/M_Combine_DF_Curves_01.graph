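# Unlit alpha-composite tool material for cloud compositing that turns an RT texture into a signed distance field for packing into a chosen channel.
# It point-samples RT, measures UV distance to the position stored in red-green, and signs it by alpha, suggesting jump-flood-style input.
# The distance is remapped around 0.5, then multiplied by Channel RGB; Channel alpha separately weights its inverse.
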Material /Volumetrics/Tools/CloudCompositing/Materials/M_Combine_DF_Curves_01
graph {
 properties {blendMode=BLEND_AlphaComposite, pixelDepthOffsetMode=PDOM_Legacy, shadingModel=MSM_Unlit, usedShadingModels=Unlit}
 nodes [
  n1 = MaterialFunctionCall!(
   UVs=TextureCoordinate()
   Texture Size=ComponentMask(None=VectorParameter() {parameterName=TextureSize, defaultValue=<rgba 256 256 0 0>}.RGB) {g=true, r=true}
  ) {materialFunction=/Engine/Functions/Engine_MaterialFunctions01/Texturing/PointSampledUVs}
  n2 = TextureSampleParameter2D!(UVs=@n1.Result, Level, Apply View MipBias) {constMipValue=0, mipValueMode=TMVM_MipLevel, parameterName=RT, texture=&Black}
  n3 = ComponentMask(None=@n2.RGB) {g=true, r=true}
  n4 = Multiply(
   A=MaterialFunctionCall(Vector 2=Subtract(A=@n1.Result, B=@n3)) {materialFunction=/Engine/Functions/Engine_MaterialFunctions02/Utility/VectorLength}.V2 Length
   B=LinearInterpolate(Alpha=@n2.A) {constA=-1}
  )
  Divide(A=@n4) {constB=0.2}
  n5 = Add(A=Multiply(A=@n4) {constB=0.5}) {constB=0.5}
  n6 = VectorParameter() {parameterName=Channel, defaultValue=<rgba 0 0 0 1>}
  LinearInterpolate(B=OneMinus(None=@n5), Alpha=@n6.A)
  Multiply(A=@n5, B=@n6.RGB)
 ]
}
