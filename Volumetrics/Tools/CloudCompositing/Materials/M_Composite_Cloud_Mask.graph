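# Unlit, two-sided alpha-composite material for the cloud compositing tool that turns a mask texture into a tinted cloud mask.
# The texture's red channel, sampled at a chosen mip, is shaped by a power, offset and perturbed by seeded, tiled gradient noise.
# The clamped, opacity-scaled result is multiplied by Color, while its ceiling forms a binary coverage scaled by inverted Alpha.
# A radial sphere mask and a stray vector add feed no other node.
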Material /Volumetrics/Tools/CloudCompositing/Materials/M_Composite_Cloud_Mask
graph {
 properties {blendMode=BLEND_AlphaComposite, pixelDepthOffsetMode=PDOM_Legacy, shadingModel=MSM_Unlit, twoSided=true, usedShadingModels=Unlit}
 nodes [
  Add(B=VectorParameter() {parameterName=asdfasfsdf, defaultValue=<rgba 5 0 0 0>}.RGB)
  n1 = TextureSampleParameter2D!(UVs, Level=ScalarParameter() {parameterName=Mip}, Apply View MipBias) {
   mipValueMode=TMVM_MipLevel
   parameterName=Texture
   samplerSource=SSM_Clamp_WorldGroupSettings
   samplerType=SAMPLERTYPE_Masks
   texture=/Engine/Functions/Engine_MaterialFunctions02/ExampleContent/Greyscale
  }
  n2 = Subtract(
   A=Power(Base=@n1.R, Exp=ScalarParameter() {defaultValue=1, parameterName=Power})
   B=Multiply(A=ScalarParameter() {defaultValue=0.5, parameterName=Noise Amount}) {constB=0.25}
  ) {constB=0.5}
  n3 = Multiply(A=TextureCoordinate(), B=ScalarParameter() {defaultValue=3, parameterName=Noise Tiling})
  n4 = ScalarParameter() {defaultValue=3, parameterName=Seed}
  n5 = Multiply(
   A=Noise!(World Position=AppendVector(A=@n3, B=@n4), FilterWidth) {noiseFunction=NOISEFUNCTION_GradientTex}
   B=ScalarParameter() {defaultValue=0.5, parameterName=Noise Amount}
  )
  n6 = Multiply(A=Max(A=Add(A=@n2, B=@n5)) {constB=0}, B=ScalarParameter() {defaultValue=1, parameterName=Opacity})
  Multiply(A=@n6, B=VectorParameter() {parameterName=Color, defaultValue=<rgba 1 1 1 0>}.RGB)
  Multiply(A=Saturate(None=Ceil(None=@n6)), B=OneMinus(None=ScalarParameter() {defaultValue=1, parameterName=Alpha}))
  n7 = Constant() {r=0.5}
  Saturate(None=SphereMask(A=TextureCoordinate(), B=@n7, Radius=@n7) {hardnessPercent=0})
 ]
}
