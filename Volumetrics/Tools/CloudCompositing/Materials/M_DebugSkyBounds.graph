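# Unlit, additive, two-sided debug material in the cloud compositing tools, named for showing sky bounds.
# It centers UVs at 0.5, takes the absolute value, scales by a scalar parameter, then floors and saturates to form a hard border mask.
# A Max node reads the R component on both inputs, so only the U-axis borders survive.
# The graph shows no link from that result to a material output.
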Material /Volumetrics/Tools/CloudCompositing/Materials/M_DebugSkyBounds
graph {
 properties {blendMode=BLEND_Additive, pixelDepthOffsetMode=PDOM_Legacy, shadingModel=MSM_Unlit, twoSided=true, usedShadingModels=Unlit}
 nodes [
  n1 = TextureCoordinate()
  n2 = Multiply(A=Abs(None=Subtract(A=@n1) {constB=0.5}), B=ScalarParameter() {defaultValue=2.05, parameterName=aasdf}) {constB=1.975}
  n3 = MaterialFunctionCall(Float2=Saturate(None=Floor(None=@n2))) {materialFunction=&BreakOutFloat2Components}
  Max(A=@n3.R, B=@n3.R)
 ]
}
