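# Unlit alpha-composite material in the cloud compositing tools that blurs a mask texture, by its name to finalize a cloud mask.
# A Custom HLSL node averages samples of the Texture parameter in a circular kernel, sized by a clamped Radius and stepped in texels from TextureSize.
# The loop accumulates inverted alpha, and a OneMinus on the node's alpha flips it back, leaving the blurred RGB unused.
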Material /Volumetrics/Tools/CloudCompositing/Materials/M_Finalize_Cloud_Mask
graph {
 properties {blendMode=BLEND_AlphaComposite, pixelDepthOffsetMode=PDOM_Legacy, shadingModel=MSM_Unlit, twoSided=true, usedShadingModels=Unlit}
 nodes [
  Constant()
  MaterialExpressionTextureObjectParameter_0 = TextureObjectParameter() {parameterName=Texture, texture=&Black}
  MaterialExpressionTextureCoordinate_1 = TextureCoordinate()
  MaterialExpressionComponentMask_0 = ComponentMask(None=VectorParameter() {parameterName=TextureSize, defaultValue=<rgba 1024 1024 0 0>}.RGB) {g=true, r=true}
  MaterialExpressionClamp_0 = Clamp(None=ScalarParameter() {defaultValue=1, parameterName=Radius}) {maxDefault=128, minDefault=1}
  n1 = Custom!(
   Tex=@MaterialExpressionTextureObjectParameter_0
   UV=@MaterialExpressionTextureCoordinate_1
   TextureSize=@MaterialExpressionComponentMask_0
   Radius=@MaterialExpressionClamp_0
  ) {
   code="\r\nfloat3 accum = 0;\r\nfloat2 texelsize = 1 / TextureSize;\r\nfloat samples = 0;\r\nfloat alpha = 0;\r\n\r\nfor(int i = -Radius;  i <= Radius; i++)\r\n{\r\n\tfor(int j = -Radius; j<= Radius; j++)\r\n\t{\r\n\r\n\t\tif(i*i + j*j < Radius * Radius)\r\n\t\t{\r\n\t\t\tfloat2 curUV = UV + (float2(i,j) * texelsize);\r\n\t\r\n\t\t\tfloat4 offsetsample = Tex.SampleLevel(TexSampler, curUV,0).rgba;\r\n\t\t\taccum += offsetsample.rgb;\r\n\t\t\tsamples += 1;\r\n\t\t\t\r\n\t\t\talpha += 1 - offsetsample.a;\r\n\t\t}\r\n\t}\r\n}\r\n\r\nif (samples > 0)\r\n{\r\n\taccum /= samples;\r\n\talpha /= samples;\r\n}\r\n\r\nreturn float4(accum, alpha);\r\n\r\n\r\n"
   description=Blur
   outputType=CMOT_Float4
   inputs [
    {
     inputName=Tex
     input {expression=&self:MaterialExpressionTextureObjectParameter_0}
    }
    {inputName=UV, input {expression=&self:MaterialExpressionTextureCoordinate_1}}
    {
     inputName=TextureSize
     input {expression=&self:MaterialExpressionComponentMask_0}
    }
    {inputName=Radius, input {expression=&self:MaterialExpressionClamp_0}}
   ]
  }
  OneMinus(None=ComponentMask(None=@n1) {a=true})
 ]
}
