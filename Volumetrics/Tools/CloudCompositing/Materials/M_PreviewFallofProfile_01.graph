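# Unlit tool material that previews a cloud layer's vertical falloff profile as a plotted curve, for cloud compositing.
# PlotFunctionOnGraph supplies x and plots GetLayerFalloff: an inverted SCurve shaped by Upper Falloff, multiplied by one minus exp(-x) of x scaled by Lower Falloff.
# The exponential is a small Custom HLSL node.
# A Cloud Bias branch, a cloud sample attribute node and a texture coordinate node do not feed the plot.
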Material /Volumetrics/Tools/CloudCompositing/Materials/M_PreviewFallofProfile_01
graph {
 properties {pixelDepthOffsetMode=PDOM_Legacy, shadingModel=MSM_Unlit, usedShadingModels=Unlit}
 nodes [
  CloudSampleAttribute()
  n1 = MaterialFunctionCall!(X range Y range=Constant4Vector() {constant=<rgba 0 1 0 1>}.RGBA, Gradient Height=Constant(), ShowGrid?, Thickness, Color, "f(x)"=@n4) {materialFunction=/Engine/Functions/Engine_MaterialFunctions01/Debug/PlotFunctionOnGraph}
  n2 = ScalarParameter() {defaultValue=1, parameterName=Upper Falloff}
  MaterialExpressionMultiply_19 = Multiply(A=ScalarParameter() {defaultValue=32, parameterName=Lower Falloff}, B=Multiply(A=@n1.x))
  n3 = Custom!(x=@MaterialExpressionMultiply_19) {
   code="return exp(-x);"
   description="exp(-x)"
   outputType=CMOT_Float1
   inputs [{inputName=x, input {expression=&self:MaterialExpressionMultiply_19}}]
  }
  n4 = Multiply(
   A=OneMinus(None=MaterialFunctionCall(In=@n1.x, Power=@n2) {materialFunction=&SCurve}.Result)
   B=OneMinus(None=@n3)
  ) {desc="GetLayerFalloff()"}
  Multiply(A=Divide(B=Max(A=@n4) {constB=0}), B=ScalarParameter() {defaultValue=0.2, parameterName=Cloud Bias})
  TextureCoordinate()
 ]
}
