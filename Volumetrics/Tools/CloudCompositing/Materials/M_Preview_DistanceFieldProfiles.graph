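# Unlit debug material in the cloud compositing tools that previews distance field profiles from a render target texture parameter, sampled at a fixed mip.
# One branch remaps RGB around a midpoint by a divisor parameter.
# Others use SphereMask on each scalar channel to isolate a thin band at a parameterized level, packing RGB bands into a float3 and tinting the alpha band yellow.
# The graph shows no branch wired to a material output.
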Material /Volumetrics/Tools/CloudCompositing/Materials/M_Preview_DistanceFieldProfiles
graph {
 properties {pixelDepthOffsetMode=PDOM_Legacy, shadingModel=MSM_Unlit, usedShadingModels=Unlit}
 nodes [
  n1 = TextureSampleParameter2D!(UVs, Level, Apply View MipBias) {constMipValue=0, mipValueMode=TMVM_MipLevel, parameterName=RT, texture=&Black}
  Divide(A=Subtract(A=@n1.RGB) {constB=0.5}, B=ScalarParameter() {defaultValue=0.1, parameterName=Divisor})
  MaterialFunctionCall(
   X=SphereMask(A=@n1.R, B=ScalarParameter() {defaultValue=0.5, parameterName=P1}, Radius=ScalarParameter() {defaultValue=0.001, parameterName=R1}) {hardnessPercent=0}
   Y=SphereMask(A=@n1.G, B=ScalarParameter() {defaultValue=0.5, parameterName=P1}, Radius=ScalarParameter() {defaultValue=0.001, parameterName=R1}) {hardnessPercent=0}
   Z=SphereMask(A=@n1.B, B=ScalarParameter() {defaultValue=0.5, parameterName=P1}, Radius=ScalarParameter() {defaultValue=0.001, parameterName=R1}) {hardnessPercent=0}
  ) {materialFunction=&MakeFloat3}
  Multiply(
   A=SphereMask(A=@n1.A, B=ScalarParameter() {defaultValue=0.5, parameterName=P1}, Radius=ScalarParameter() {defaultValue=0.001, parameterName=R1}) {hardnessPercent=0}
   B=Constant3Vector() {constant=<rgba 1 1 0 0>}
  )
 ]
}
