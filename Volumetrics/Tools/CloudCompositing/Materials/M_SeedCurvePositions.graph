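# Unlit alpha-composite material for the volumetric cloud compositing tool that computes seed positions along a height curve read from the RT texture.
# Samples curve height at U and one Res texel either side, and a Custom node returns closest points on the two thin segments joining them.
# Nonzero points are Y-flipped and masked to the area under the curve, alongside an inverse mask.
# Separate Rise/Run branches hold slope-intersection math and debug line drawing.
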Material /Volumetrics/Tools/CloudCompositing/Materials/M_SeedCurvePositions
graph {
 properties {blendMode=BLEND_AlphaComposite, pixelDepthOffsetMode=PDOM_Legacy, shadingModel=MSM_Unlit, usedShadingModels=Unlit}
 nodes [
  n1 = ComponentMask(None=TextureCoordinate()) {r=true}
  n2 = TextureSampleParameter2D(UVs=AppendVector(A=@n1, B=Constant() {r=0.5})) {
   parameterName=RT
   samplerType=SAMPLERTYPE_LinearColor
   texture=/Engine/Functions/Engine_MaterialFunctions02/PivotPainter2/Black_1x1_EXR_Texture
  }
  n3 = MaterialFunctionCall(Float2=ComponentMask(None=@n2.RGB) {g=true, r=true}) {materialFunction=&BreakOutFloat2Components}
  AppendVector(A=@n3.R, B=OneMinus(None=@n3.G))
  MaterialExpressionScalarParameter_4 = ScalarParameter() {defaultValue=0.5, parameterName=Rise}
  MaterialExpressionScalarParameter_3 = ScalarParameter() {defaultValue=0.3, parameterName=Run}
  n4 = SquareRoot(None=Add(
    A=Multiply(A=@MaterialExpressionScalarParameter_4, B=@MaterialExpressionScalarParameter_4)
    B=Multiply(A=@MaterialExpressionScalarParameter_3, B=@MaterialExpressionScalarParameter_3)
   )) {desc=c}
  n5 = Divide(A=@n4, B=@MaterialExpressionScalarParameter_3) {desc=c / b}
  n6 = Divide(
   A=Divide(A=@MaterialExpressionScalarParameter_3, B=@n5) {desc=e}
   B=Divide(A=@MaterialExpressionScalarParameter_3, B=@MaterialExpressionScalarParameter_4) {desc=b / a}
  ) {desc="e / (b/a)"}
  AppendVector(
   A=Multiply(A=Divide(A=@MaterialExpressionScalarParameter_4, B=@n4) {desc=a / c}, B=@n6) {desc=g}
   B=Multiply(A=Divide(B=@n5) {desc=b / c}, B=@n6) {desc=f}
  )
  MaterialExpressionTextureCoordinate_1 = TextureCoordinate()
  MaterialExpressionScalarParameter_6 = ScalarParameter() {defaultValue=0.5, parameterName=Rise}
  MaterialExpressionScalarParameter_5 = ScalarParameter() {defaultValue=0.3, parameterName=Run}
  n7 = Custom!(a=@MaterialExpressionScalarParameter_4, b=@MaterialExpressionScalarParameter_3) {
   code="float c = sqrt(a*a + b*b);\r\n\r\nfloat e = b / (c/b);\r\nfloat d  = e / (b/a);\r\n\r\nfloat g = (a/c) * d;\r\nfloat f = (b/c) * d;\r\n\r\nreturn float2(g, f);"
   description=Slope to Line intersection
   outputType=CMOT_Float2
   inputs [
    {inputName=a, input {expression=&self:MaterialExpressionScalarParameter_4}}
    {inputName=b, input {expression=&self:MaterialExpressionScalarParameter_3}}
   ]
  }
  MaterialExpressionMaterialFunctionCall_1 = MaterialFunctionCall(Float2=@n7) {materialFunction=&BreakOutFloat2Components}
  Custom!(
   UV=@MaterialExpressionTextureCoordinate_1
   Rise=@MaterialExpressionScalarParameter_6
   Run=@MaterialExpressionScalarParameter_5
   g=@MaterialExpressionMaterialFunctionCall_1.R
   f=@MaterialExpressionMaterialFunctionCall_1.R
  ) {
   code="//Debug Lines for NearestLine Intersection\r\n\r\n\r\n\r\nstruct ShapeDrawer\r\n{\r\n\r\n\tfloat2 P;\r\n\t\r\n\tfloat DrawLine(float2 inA, float2 inB, float inWidth)\r\n\t{\r\n\t\tfloat2 UV = P - inA;\r\n\t\tfloat2 squashvec = inB - inA;\r\n\t\t\r\n\t\t\r\n\t\t\r\n\t\tfloat2 distalongline =  saturate( dot(UV, squashvec) / dot(squashvec, squashvec)) * squashvec;\r\n\t\t\r\n\t\tfloat2 squashUV = UV - distalongline;\r\n\t\t\r\n\t\tfloat thisline = 1-length(squashUV);\r\n\t\tthisline = saturate (floor( thisline + inWidth));\r\n\t\t\r\n\t\treturn thisline;\r\n\t}\r\n\t\r\n\tfloat DrawCircle(float2 inC, float inRadius, float inWidth)\r\n\t{\r\n\t\tfloat2 UV = P - inC;\r\n\t\r\n\t\tfloat dist = 1 - abs( length(UV) - inRadius);\r\n\t\tdist = saturate( floor( dist + inWidth) );\r\n\t\treturn dist;\r\n\t}\r\n\t\r\n\r\n\t\r\n};\r\n\r\n\r\nShapeDrawer Shapey;\r\nShapey.P = float2( UV.x, 1 - UV.y);\r\n\r\nfloat3 output = 0;\r\n\r\n\r\noutput += Shapey.DrawLine(0, float2(0,Rise), 0.01);\r\noutput += Shapey.DrawLine(0, float2(Run,0), 0.01);\r\noutput += Shapey.DrawLine(0, float2(Run,Rise), 0.01);\r\n\r\noutput += Shapey.DrawLine(float2(0,Rise), float2(0,Rise)+float2(g,-f), 0.01);\r\n\r\n\r\n\r\n\r\n\r\nreturn output;"
   outputType=CMOT_Float1
   inputs [
    {inputName=UV, input {expression=&self:MaterialExpressionTextureCoordinate_1}}
    {inputName=Rise, input {expression=&self:MaterialExpressionScalarParameter_6}}
    {inputName=Run, input {expression=&self:MaterialExpressionScalarParameter_5}}
    {inputName=g, input {expression=&self:MaterialExpressionMaterialFunctionCall_1}}
    {
     inputName=f
     input {expression=&self:MaterialExpressionMaterialFunctionCall_1, outputIndex=1}
    }
   ]
  }
  MaterialExpressionTextureCoordinate_9 = TextureCoordinate()
  MaterialExpressionAppendVector_8 = AppendVector(A=@n1, B=@n2.G)
  n8 = Divide(B=ScalarParameter() {defaultValue=256, parameterName=Res})
  n9 = Add(A=@n1, B=@n8)
  n10 = TextureSampleParameter2D(UVs=AppendVector(A=@n9, B=Constant() {r=0.5})) {
   parameterName=RT
   samplerType=SAMPLERTYPE_LinearColor
   texture=&Black_1x1_EXR_Texture
  }
  MaterialExpressionAppendVector_9 = AppendVector(A=@n9, B=@n10.G)
  n11 = Subtract(A=@n1, B=@n8)
  n12 = TextureSampleParameter2D(UVs=AppendVector(A=@n11, B=Constant() {r=0.5})) {
   parameterName=RT
   samplerType=SAMPLERTYPE_LinearColor
   texture=&Black_1x1_EXR_Texture
  }
  MaterialExpressionAppendVector_10 = AppendVector(A=@n11, B=@n12.G)
  n13 = Custom!(
   UV=@MaterialExpressionTextureCoordinate_9
   a=@MaterialExpressionAppendVector_8
   b=@MaterialExpressionAppendVector_9
   c=@MaterialExpressionAppendVector_10
  ) {
   code="//Debug Lines for NearestLine Intersection\r\n\r\n\r\n\r\nstruct ShapeDrawer\r\n{\r\n\r\n\tfloat2 P;\r\n\t\r\n\tfloat2 DrawLine(float2 inA, float2 inB, float inWidth)\r\n\t{\r\n\t\tfloat2 UV = P - inA;\r\n\t\tfloat2 squashvec = inB - inA;\r\n\t\tfloat2 distalongline =  saturate( dot(UV, squashvec) / dot(squashvec, squashvec)) * squashvec;\r\n\t\tfloat2 squashUV = UV - distalongline;\r\n\t\t\r\n\t\tfloat thisline = 1-length(squashUV);\r\n\t\tthisline = saturate (floor( thisline + inWidth));\r\n\t\t\r\n\t\t\r\n\t\t\r\n\t\treturn thisline * lerp(inA,inB,  saturate( dot(UV, squashvec) / dot(squashvec, squashvec)) ) ;\r\n\t}\r\n\t\r\n\tfloat DrawCircle(float2 inC, float inRadius, float inWidth)\r\n\t{\r\n\t\tfloat2 UV = P - inC;\r\n\t\r\n\t\tfloat dist = 1 - abs( length(UV) - inRadius);\r\n\t\tdist = saturate( floor( dist + inWidth) );\r\n\t\treturn dist;\r\n\t}\r\n\t\r\n\r\n\t\r\n};\r\n\r\n\r\nShapeDrawer Shapey;\r\nShapey.P = float2( UV.x, 1 - UV.y);\r\n\r\nfloat mindist = 1000000;\r\nfloat2 output = 0;\r\n\r\nfloat2 testline = Shapey.DrawLine(a,b, 0.01);\r\nfloat testdist = length(Shapey.P - testline);\r\n\r\nif (testdist < mindist)\r\n{\r\n\tmindist = testdist;\r\n\toutput = testline;\r\n}\r\n\r\ntestline = Shapey.DrawLine(c,a, 0.01);\r\ntestdist = length(Shapey.P - testline);\r\n\r\nif (testdist < mindist)\r\n{\r\n\tmindist = testdist;\r\n\toutput = testline;\r\n}\r\n\r\n\r\nreturn output;"
   outputType=CMOT_Float2
   inputs [
    {inputName=UV, input {expression=&self:MaterialExpressionTextureCoordinate_9}}
    {inputName=a, input {expression=&self:MaterialExpressionAppendVector_8}}
    {inputName=b, input {expression=&self:MaterialExpressionAppendVector_9}}
    {inputName=c, input {expression=&self:MaterialExpressionAppendVector_10}}
   ]
  }
  n14 = ComponentMask(None=@n13) {r=true}
  n15 = MaterialFunctionCall(Float2=@n13) {materialFunction=&BreakOutFloat2Components}
  n16 = AppendVector(A=@n15.R, B=OneMinus(None=@n15.G))
  n17 = Constant()
  n18 = Constant()
  n19 = OneMinus(None=ComponentMask(None=TextureCoordinate()) {g=true})
  n20 = Saturate(None=Ceil(None=Subtract(A=@n2.G, B=@n19)))
  Multiply(A=If(A=@n15.G, "A > B"=If(A=@n14, "A > B"=@n16, "A == B"=@n17, "A < B"=@n17), "A == B"=@n18, "A < B"=@n18), B=@n20)
  OneMinus(None=@n20)
 ]
}
