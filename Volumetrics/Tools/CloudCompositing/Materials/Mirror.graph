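# Mirror is a Default Lit material in the Volumetrics cloud compositing tools.
# It holds two scalar parameters, R and S, with S defaulting to one.
# No connections from the parameters to material outputs appear in the graph.
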Material /Volumetrics/Tools/CloudCompositing/Materials/Mirror
graph {
 properties {pixelDepthOffsetMode=PDOM_Legacy, usedShadingModels=Default Lit}
 nodes [
  ScalarParameter() {parameterName=R}
  ScalarParameter() {defaultValue=1, parameterName=S}
 ]
}
