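# Unlit material in the cloud compositing tools that generates animated lightning flash intensity, with no material output connection shown.
# A Custom node sums jittered, noise-timed radial flashes over neighbouring cells and four layers in flipbook-encoded volume coordinates.
# A separate texture branch dots a time-shifted, thresholded lightning pattern with the same texture sampled over time, so each channel strikes independently, then scales the result.
# Leftover debug-value and UV nodes remain.
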Material /Volumetrics/Tools/CloudCompositing/Materials/RenderLightning
graph {
 properties {pixelDepthOffsetMode=PDOM_Legacy, shadingModel=MSM_Unlit, usedShadingModels=Unlit}
 nodes [
  Constant() {r=1}
  n1 = ComponentMask(None=VectorParameter() {parameterName=XYFrames, defaultValue=<rgba 8 8 0 0>}.RGB) {g=true, r=true}
  MaterialExpressionAdd_2 = Add(A=Multiply(
    A=MaterialFunctionCall!(UVs, XY Frames=@n1) {materialFunction=/Volumetrics/Content/VolumeTextures/MaterialFunctions/EncodeVolumeCoordinates}.Result
    B=ScalarParameter() {defaultValue=5, parameterName=Grid Cells}
   ))
  MaterialExpressionScalarParameter_2 = ScalarParameter() {defaultValue=0.5, parameterName=Light Radius}
  MaterialExpressionTextureObject_0 = TextureObject() {texture=/Engine/MapTemplates/TilingNoise05}
  MaterialExpressionScalarParameter_3 = ScalarParameter() {defaultValue=0.2, parameterName=Threshold}
  MaterialExpressionMaterialFunctionCall_3 = MaterialFunctionCall(Speed=ScalarParameter() {parameterName=Speed}) {materialFunction=&TimeWithSpeedVariable}
  MaterialExpressionScalarParameter_5 = ScalarParameter() {defaultValue=2, parameterName=Mip}
  MaterialExpressionScalarParameter_7 = ScalarParameter() {defaultValue=5, parameterName=NumCells}
  n2 = Custom!(
   inCell=@MaterialExpressionAdd_2
   LightRadius=@MaterialExpressionScalarParameter_2
   Noise=@MaterialExpressionTextureObject_0
   Threshold=@MaterialExpressionScalarParameter_3
   Time=@MaterialExpressionMaterialFunctionCall_3.Time
   Mip=@MaterialExpressionScalarParameter_5
   NumCells=@MaterialExpressionScalarParameter_7
  ) {
   code="\r\nfloat accum = 0;\r\n\r\nfor (int h = -1; h <= 1; h++)\r\n{\r\n\r\n\tfor (int j = -1; j <= 1; j++)\r\n\t{\t\r\n\r\n\t\tfloat2 relativeidx = float2(h, j);\r\n\r\n\t\tfor (int i = 0; i < 4; i++)\r\n\t\t{\t\r\n\t\t\tfloat3 cell = float3(inCell.xy, inCell.z + i + (Time * 5.0));\r\n\t\t\tcell .xy += relativeidx;\r\n\t\t\t\r\n\t\t\tcell.xy = fmod(cell.xy, NumCells);\r\n\t\t\t\r\n\t\t\tfloat3 curnoise = MaterialExpressionVectorNoise( cell, 1, 0, 0, 300 ).xyz;\r\n\t\t\tcurnoise.xy -= 0.5;\r\n\t\t\t\r\n\t\t\tfloat2 center = floor(inCell.xy) + 0.5;\r\n\t\t\tcenter += relativeidx;\r\n\t\t\t\r\n\t\t\tfloat mask = length(inCell.xy - (center + curnoise));\r\n\t\t\t\r\n\t\t\tfloat thisradius = lerp(LightRadius * 0.75, LightRadius * 1.25, curnoise.z);\r\n\t\t\tmask = 1 - saturate( mask / LightRadius);\r\n\t\t\t\r\n\t\t\tfloat2 timeUV = float2(curnoise.x, curnoise.z + (inCell.z / NumCells));\r\n\t\t\t\r\n\t\t\tfloat timeline = Noise.SampleLevel(NoiseSampler, timeUV, Mip);\r\n\t\t\ttimeline = saturate( (timeline - Threshold) / (1.0  - Threshold) );\r\n\t\t\t\r\n\t\t\t\r\n\t\t\taccum += mask * timeline * timeline;\r\n\t\t}\r\n\t}\r\n}\r\n\r\n\r\nreturn accum;"
   description=inCell
   inputs [
    {inputName=inCell, input {expression=&self:MaterialExpressionAdd_2}}
    {inputName=LightRadius, input {expression=&self:MaterialExpressionScalarParameter_2}}
    {inputName=Noise, input {expression=&self:MaterialExpressionTextureObject_0}}
    {inputName=Threshold, input {expression=&self:MaterialExpressionScalarParameter_3}}
    {
     inputName=Time
     input {expression=&self:MaterialExpressionMaterialFunctionCall_3, outputIndex=1}
    }
    {inputName=Mip, input {expression=&self:MaterialExpressionScalarParameter_5}}
    {inputName=NumCells, input {expression=&self:MaterialExpressionScalarParameter_7}}
   ]
  }
  Divide(A=Multiply(A=@n2, B=@n2), B=ScalarParameter() {defaultValue=0.2, parameterName=Divisor})
  Frac(None=Add(A=TextureCoordinate(), B=ScalarParameter() {parameterName=Offset}) {constB=0.5})
  Frac()
  MaterialFunctionCall!(
   Vector3=Fmod(A=ScalarParameter() {defaultValue=4, parameterName=asdf}, B=@MaterialExpressionScalarParameter_7)
   MaximumNumberOfDigits
   UVs
   DebugTextLocation RG_UpperRight BA_LowerLeft
   Component Spacing
  ) {materialFunction=/Engine/Functions/Engine_MaterialFunctions02/Utility/DebugFloat3Values}
  n3 = MaterialFunctionCall(Speed=ScalarParameter() {parameterName=Pattern Shift SPeed}) {materialFunction=&TimeWithSpeedVariable}
  n4 = Multiply(
   A=Subtract(A=TextureSample!(UVs=@n3.Frac Time, Tex, Level, Apply View MipBias) {mipValueMode=TMVM_MipLevel, samplerType=SAMPLERTYPE_LinearColor, texture=/Engine/EngineMaterials/RandomAngles}.RGB) {constB=0.5}
   B=ScalarParameter() {defaultValue=0.5, parameterName=Shift}
  )
  n5 = TextureSample(UVs=Add(A=TextureCoordinate(), B=ComponentMask(None=@n4) {g=true, r=true}), Level=Constant() {r=1}) {
   constMipValue=0
   mipValueMode=TMVM_MipLevel
   samplerSource=SSM_Wrap_WorldGroupSettings
   texture=/Volumetrics/Tools/CloudCompositing/Textures/T_LightingPattern_01
  }
  n6 = ScalarParameter() {defaultValue=0.2, parameterName=Lightning Min}
  n7 = TextureSample(UVs=MaterialFunctionCall(Speed=ScalarParameter() {defaultValue=0.3, parameterName=Lightning Speed}) {materialFunction=&TimeWithSpeedVariable}.Frac Time) {
   constMipValue=0
   mipValueMode=TMVM_MipLevel
   samplerSource=SSM_Wrap_WorldGroupSettings
   texture=&T_LightingPattern_01
  }
  n8 = ScalarParameter() {defaultValue=0.3, parameterName=Strike Minimum}
  n9 = Divide(A=Subtract(A=@n7.RGB, B=@n8), B=OneMinus(None=@n8))
  n10 = DotProduct(A=Max(A=Subtract(A=@n5.RGB, B=@n6)) {constB=0}, B=Max(A=Multiply(A=@n9)) {constB=0})
  MaterialFunctionCall(f3=Multiply(A=Max(A=@n10) {constB=0}, B=ScalarParameter() {defaultValue=2, parameterName=Lightning Amount})) {materialFunction=&AddComponents}
  TextureCoordinate()
 ]
}
